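# Translucent UI material for the smart pistol reticle's ammo bar, drawn as a cyan, segmented arc on a circular outline.
# Rotated radial-segment functions cut tick marks and a section mask, combined with an SDF circle outline to form the shape.
# A radial angle gradient compared against the scaled Progress parameter, sharpened by Fill_Hardness, switches opacity between backing and fill alphas.
# Section_Orientation rotates the section mask and, inverted, the fill gradient.
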Material /Smartgun/Reticle/Materials/M_UI_SmartPistolAmmoBar
graph {
 properties {blendMode=BLEND_Translucent, materialDomain=MD_UI, pixelDepthOffsetMode=PDOM_Legacy, usedShadingModels=Default Lit}
 nodes [
  Constant() {r=0.8686}
  Constant() {r=0.4554}
  Constant() {r=60}
  Constant() {r=0.0266}
  n1 = MaterialFunctionCall!(Rotation_Fraction=Constant() {r=0.8686}, TexCoord=TextureCoordinate()) {materialFunction=/FortUILibrary/Foundation/Art_ToolKit/AngledBoxes/MF_UI_Rotator_Fractions}
  n2 = Constant() {r=60}
  n3 = Constant() {r=0.4554}
  n4 = SmoothStep(Value=MaterialFunctionCall(Size=Constant() {r=0.75}, OutlineThickness=Constant() {r=0.04}) {materialFunction=&MF_UI_Circle}.SDFOutline) {constMax=0.01}
  n5 = MaterialFunctionCall(Rotation_Fraction=ScalarParameter() {defaultValue=0.825, group=Section, parameterName=Section_Orientation}, TexCoord=TextureCoordinate()) {materialFunction=&MF_UI_Rotator_Fractions}
  n6 = Constant() {r=1}
  n7 = Multiply(
   A=SmoothStep(Value=MaterialFunctionCall!(UV=@n1.Result, Segment Count=@n2, Segment Spacing=@n3, Fixed Spacing) {materialFunction=/FortUILibrary/MaterialFunctions/Basics/MF_UI_RadialSegments}.Segmented Triangles Gradient) {constMax=0.3}
   B=Multiply(
    A=OneMinus(None=@n4)
    B=MaterialFunctionCall(UV=@n5.Result, Segment Count=@n6) {materialFunction=&MF_UI_RadialSegments}.Segmented Triangles Fill
   )
  )
  n8 = ScalarParameter() {defaultValue=0.825, group=Section, parameterName=Section_Orientation}
  n9 = Subtract(
   A=MaterialFunctionCall(Rotation_Fraction=OneMinus(None=@n8), TexCoord=TextureCoordinate()) {materialFunction=&MF_UI_Rotator_Fractions}.Result
   B=Constant() {r=0.5}
  )
  n10 = Saturate(None=MaterialFunctionCall(Vector or UVs=Multiply(A=@n9) {constB=2}) {materialFunction=&VectorToRadialValue}.Vector Converted to Angle)
  n11 = Saturate(None=Divide(A=ScalarParameter() {defaultValue=1, parameterName=Progress, sliderMax=1}, B=Constant() {r=5.0608}) {constB=4})
  n12 = Saturate(None=Multiply(A=Subtract(A=@n10, B=@n11), B=ScalarParameter() {defaultValue=256, parameterName=Fill_Hardness}))
  Multiply(
   A=@n7
   B=LinearInterpolate(
    A=ScalarParameter() {defaultValue=0.25, parameterName=Fuel_BackingAlpha}
    B=ScalarParameter() {defaultValue=1, parameterName=Fuel_FillAlpha}
    Alpha=OneMinus(None=@n12)
   )
  )
  Multiply(A=Constant3Vector() {constant=<rgba 0 0.7232 1 0>}, B=@n7)
 ]
}
