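# Translucent UI material that draws a glowing circle outline for the smart pistol reticle, tinted by lock-on state.
# Three MF_UI_Circle outlines share size, thickness and blur parameters, with named UV reroutes providing unshifted and positive and negative OffsetMax coordinates.
# The masks are packed into separate RGB channels for a split-channel fringe, multiplied by a normal-to-locked color blend driven by IsLocked.
# Summed masks scaled by Opacity form the alpha.
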Material /Smartgun/Reticle/Materials/M_UI_SmartPistolCircle
graph {
 properties {blendMode=BLEND_Translucent, materialDomain=MD_UI, pixelDepthOffsetMode=PDOM_Legacy, usedShadingModels=Default Lit}
 nodes [
  MaterialExpressionAppend3Vector_0 = MaterialXAppend3Vector!(A=NamedRerouteUsage(), B=NamedRerouteUsage(), C=NamedRerouteUsage())
  Multiply(
   A=@MaterialExpressionAppend3Vector_0
   B=LinearInterpolate(
    A=VectorParameter() {parameterName=CircleColorNormal, defaultValue=<rgba 0.0206 0.2519 0.4635 1>}.RGB
    B=VectorParameter() {parameterName=CircleColorLockedOn, defaultValue=<rgba 0.0037 0.6004 1 1>}.RGB
    Alpha=ScalarParameter() {defaultValue=1, parameterName=IsLocked}
   )
  )
  n1 = TextureCoordinate()
  NamedRerouteDeclaration(None=Add(A=@n1, B=ScalarParameter() {defaultValue=0.045, parameterName=OffsetMax})) {name=Offset1, nodeColor=<rgba 0 0 0 1>}
  n2 = ScalarParameter() {defaultValue=0.045, parameterName=OffsetMax}
  NamedRerouteDeclaration(None=Add(A=@n1, B=Multiply(A=@n2) {constB=-1})) {name=Offset3, nodeColor=<rgba 0 0 0 1>}
  NamedRerouteDeclaration(None=@n1) {name=UV, nodeColor=<rgba 0 0 0 1>}
  NamedRerouteDeclaration(None=MaterialFunctionCall(
    UVs=NamedRerouteUsage()
    Size=ScalarParameter() {defaultValue=0.8, parameterName=Size}
    OutlineThickness=ScalarParameter() {defaultValue=0.0001, parameterName=OutlineThickness}
    GlowMax=ScalarParameter() {defaultValue=0.1, parameterName=Blur}
   ) {materialFunction=&MF_UI_Circle}.Outline) {name=Circle1, nodeColor=<rgba 0 0 0 1>}
  NamedRerouteDeclaration(None=MaterialFunctionCall(
    UVs=NamedRerouteUsage()
    Size=ScalarParameter() {defaultValue=0.8, parameterName=Size}
    OutlineThickness=ScalarParameter() {defaultValue=0.0001, parameterName=OutlineThickness}
    GlowMax=ScalarParameter() {defaultValue=0.1, parameterName=Blur}
   ) {materialFunction=&MF_UI_Circle}.Outline) {name=Circle2, nodeColor=<rgba 0 0 0 1>}
  NamedRerouteDeclaration(None=MaterialFunctionCall(
    UVs=NamedRerouteUsage()
    Size=ScalarParameter() {defaultValue=0.8, parameterName=Size}
    OutlineThickness=ScalarParameter() {defaultValue=0.0001, parameterName=OutlineThickness}
    GlowMax=ScalarParameter() {defaultValue=0.1, parameterName=Blur}
   ) {materialFunction=&MF_UI_Circle}.Outline) {name=Circle3, nodeColor=<rgba 0 0 0 1>}
  n3 = Add(A=NamedRerouteUsage(), B=NamedRerouteUsage())
  n4 = NamedRerouteUsage()
  Reroute(None=Multiply(A=Add(A=@n3, B=@n4), B=ScalarParameter() {defaultValue=0.4, parameterName=Opacity}))
 ]
}
