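# Translucent UI material for the Smart Pistol reticle, apparently drawing one glowing pip of a stack indicator as a thin bar.
# MF_UI_Box SDFs pass through inverted SmoothSteps to form soft box and glow masks, shared through named reroutes.
# PipColor blends FilledColor toward white, and PulseAnim drives a lerp toward a pale pink.
# The masked color blends with a tinted inner-box glow, while an Opacity parameter scales the alpha mask.
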Material /Smartgun/Reticle/Materials/M_UI_SmartPistolStack
graph {
 properties {blendMode=BLEND_Translucent, materialDomain=MD_UI, pixelDepthOffsetMode=PDOM_Legacy, usedShadingModels=Default Lit}
 nodes [
  NamedRerouteDeclaration(None=LinearInterpolate(
    A=VectorParameter() {parameterName=FilledColor, defaultValue=<rgba 1 0 0.8963 1>}.RGB
    B=Constant3Vector() {constant=<rgba 1 1 1 0>}
    Alpha=ScalarParameter() {defaultValue=0.1, parameterName=PipColor}
   ) {constAlpha=0}) {name=Colors, nodeColor=<rgba 0 0 0 1>}
  NamedRerouteDeclaration(None=MaterialFunctionCall(Size=Constant2Vector() {g=0.0071, r=0.4091}) {materialFunction=&MF_UI_Box}.SDF) {name=RoundedBoxSDF, nodeColor=<rgba 0 0 0 1>}
  n1 = Subtract(A=NamedRerouteUsage(), B=Constant() {r=0.3})
  NamedRerouteDeclaration(None=OneMinus(None=SmoothStep(Value=@n1) {constMax=0.4})) {name=Box, nodeColor=<rgba 0 0 0 1>}
  n2 = Subtract(A=NamedRerouteUsage(), B=Constant() {r=0.02})
  NamedRerouteDeclaration(None=OneMinus(None=SmoothStep(Value=@n2) {constMax=0.3})) {name=CenterGlow, nodeColor=<rgba 0 0.2969 1 1>}
  n3 = NamedRerouteUsage()
  Saturate(None=Multiply(A=@n3, B=ScalarParameter() {defaultValue=1, parameterName=Opacity}))
  n4 = Constant3Vector() {constant=<rgba 1 0.6536 0.8774 0>}
  n5 = NamedRerouteDeclaration(None=MaterialFunctionCall(Size=Constant2Vector() {g=0.0071, r=0.4091}, GlowMin=Constant() {r=-2}, GlowMax=Constant() {r=0.1}) {materialFunction=&MF_UI_Box}.SDF) {name=RoundedBoxInner, nodeColor=<rgba 0 0 0 1>}
  n6 = Constant() {r=0.01}
  n7 = OneMinus(None=SmoothStep(Value=Subtract(A=@n5, B=@n6)) {constMax=0.38})
  n8 = NamedRerouteUsage()
  n9 = Constant3Vector() {constant=<rgba 1 0.7109 0.9302 0>}
  n10 = ScalarParameter() {parameterName=PulseAnim}
  n11 = Multiply(A=Multiply(A=LinearInterpolate(A=@n8, B=@n9, Alpha=@n10), B=@n3), B=@n3)
  Saturate(None=LinearInterpolate(A=Multiply(A=@n4, B=@n7), B=Saturate(None=@n11)) {constAlpha=0.3})
 ]
}
