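# Reticle widget for the auto-aim smart pistol, showing stacking lock-ons and magazine ammo.
# Tick lazily binds to the pawn's AutoAim component and, while targeting, shows the reticle and fills four pips by lock-on count via material swaps.
# Component events play lock-on animations forward or reversed, plus a pulse on full charge.
# Ammo changes drive the bar material's Progress; TickSpread converts spread and FOV into crosshair offsets, and OnPaint draws a circle.
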WidgetBlueprint /Smartgun/Reticle/WBP_Reticle_StackingLockons
animations [
 OnLockOn {
  binding_count=2
  length_seconds=0.5
  object=&self:OnLockOn
  sequence {bindings [
    ReticleVisual {
     id=#1
     possessed_class=Image
     tracks [{
       type=MovieSceneWidgetMaterialTrack
       sections [{
         type=MovieSceneParameterSection
         end_frame=10
         start_frame=0
         channels [
          Opacity {type=MovieSceneScriptingFloatChannel, keys [[0,0.1011],[6.9883,0.1778],[9.9897,0.2]]}
          Blur {type=MovieSceneScriptingFloatChannel, keys [[0,0.13],[9.9897,0.0219],[10.1927,0]]}
          OutlineThickness {type=MovieSceneScriptingFloatChannel, keys [[0,0.0001],[9.9897,0.0012],[10,0.001],[10.1927,0.0101]]}
          IsLocked {type=MovieSceneScriptingFloatChannel, keys [[0,0],[9.9897,1]]}
          OffsetMax {type=MovieSceneScriptingFloatChannel, keys [[0,0.0199],[6.4867,0.0097],[10,-0.0012]]}
         ]
        }]
      }]
    }
    LockedOnAmmo {
     id=#2
     possessed_class=Image
     tracks [{
       type=MovieSceneWidgetMaterialTrack
       sections [{type=MovieSceneParameterSection, end_frame=9, start_frame=0, channels [Opacity {type=MovieSceneScriptingFloatChannel, const [0.1034,2]}]}]
      }]
    }
   ]}
 }
 FinalLock {
  binding_count=4
  length_seconds=2.0075
  object=&self:FinalLock
  sequence {bindings [
    LockedOnAmmo {
     id=#3
     possessed_class=Image
     tracks [{
       type=MovieSceneWidgetMaterialTrack
       sections [{
         type=MovieSceneParameterSection
         end_frame=39
         start_frame=-1
         channels [PulseAnim {type=MovieSceneScriptingFloatChannel, keys [[0.0463,0],[9.8207,1],[19.5713,0],[29.3217,1],[39.9433,0]]}]
        }]
      }]
    }
    LockedOnAmmo_1 {
     id=#4
     possessed_class=Image
     tracks [{
       type=MovieSceneWidgetMaterialTrack
       sections [{
         type=MovieSceneParameterSection
         end_frame=40
         start_frame=0
         channels [
          PulseAnim {type=MovieSceneScriptingFloatChannel, keys [[0,0],[10.0043,1],[19.984,0],[29.9637,1],[39.9433,0]]}
          FilledColorR {type=MovieSceneScriptingFloatChannel, const [1,5]}
          FilledColorG {type=MovieSceneScriptingFloatChannel, keys [[0.1103,0],[9.9883,1],[20.0257,0],[29.9353,1],[40.0043,0]]}
          FilledColorB {type=MovieSceneScriptingFloatChannel, keys [[0.1103,0.8963],[9.9883,1],[20.0257,0.8963],[29.9353,1],[40.0043,0.8963]]}
          FilledColorA {type=MovieSceneScriptingFloatChannel, const [1,3]}
         ]
        }]
      }]
    }
    LockedOnAmmo_3 {
     id=#5
     possessed_class=Image
     tracks [{
       type=MovieSceneWidgetMaterialTrack
       sections [{
         type=MovieSceneParameterSection
         end_frame=39
         start_frame=0
         channels [PulseAnim {type=MovieSceneScriptingFloatChannel, keys [[0,0],[10.0043,1],[19.984,0],[29.9637,1],[39.9433,0]]}]
        }]
      }]
    }
    LockedOnAmmo_2 {
     id=#6
     possessed_class=Image
     tracks [{
       type=MovieSceneWidgetMaterialTrack
       sections [{
         type=MovieSceneParameterSection
         end_frame=39
         start_frame=0
         channels [PulseAnim {type=MovieSceneScriptingFloatChannel, keys [[0,0],[10.0043,1],[19.984,0],[29.9637,1],[39.9433,0]]}]
        }]
      }]
    }
   ]}
 }
]
graphs [
 EventGraph {nodes [
   Event Construct()
   > Add(
    TargetArray:Array of Image Object References=Get AmmoList() -> (AmmoList:Array of Image Object References)
    NewItem:"Image Object Reference (by ref)"=Get LockedOnAmmo_1() -> (LockedOnAmmo_1:Image Object Reference)
   ) {is=title K2Node_CallArrayFunction}
   > Add(TargetArray=Get AmmoList(), NewItem=Get LockedOnAmmo_2() -> (LockedOnAmmo_2:Image Object Reference))
   > Add(TargetArray=Get AmmoList(), NewItem=Get LockedOnAmmo_3() -> (LockedOnAmmo_3:Image Object Reference))
   > Add(TargetArray=Get AmmoList(), NewItem=Get LockedOnAmmo() -> (LockedOnAmmo:Image Object Reference))
   > n1 = Cast To B_AutoAim_Pistol!(
    execute
    Object=GetEquippedWeapon!(self:Fort Weapon Reticle Extension Widget Base Object Reference) -> (ReturnValue:Fort Weapon Object Reference)
   ) -> (then, CastFailed, AsB Auto Aim Pistol:B Auto Aim Pistol Object Reference) {is=title K2Node_DynamicCast}
   > n2 = Set As B Pistol Smart Final(As B Pistol Smart Final:B Auto Aim Pistol Object Reference=@n1) -> (Output_Get:B Auto Aim Pistol Object Reference)
   > Bind Event to On Local Ammo Changed(self:Fort Weapon Object Reference=@n2, Delegate=Create Event()) {is=title K2Node_AddDelegate}
   n3 = Get As B Pistol Smart Final() -> (As B Pistol Smart Final:B Auto Aim Pistol Object Reference)
   > ResetAmmo!(
    execute
    self
    LocalCount:int=GetMagazineAmmoCount!(self:Fort Weapon Object Reference=@n3) -> (ReturnValue:int=0)
    LocalRemaining:int=GetRemainingAmmo!(self:Fort Weapon Object Reference=@n3) -> (ReturnValue:int=0)
   ) -> (then)
   Event Tick()
   > UpdateBinding()
   n4 = GetEquippedWeapon()
   > Is Valid(InputObject=@n4)
   > Is Valid(InputObject=Get AutoAimComponent() -> (AutoAimComponent:Auto Aim Pistol Component Object Reference))
   > n5 = Branch(Condition=GetIsTargeting!(self:Fort Weapon Object Reference=@n4) -> (ReturnValue:bool))
   > SetColorAndOpacity!(
    execute
    self:Image Object Reference=[Get LockedOnAmmo_1(), Get LockedOnAmmo(), Get LockedOnAmmo_2(), Get LockedOnAmmo_3()]
    InColorAndOpacity_R:float=1
    InColorAndOpacity_G:float=1
    InColorAndOpacity_B:float=1
    InColorAndOpacity_A:float=1
   ) -> (then)
   > SetColorAndOpacity(
    self=Get ReticleVisual() -> (ReticleVisual:Image Object Reference)
    InColorAndOpacity_R=1
    InColorAndOpacity_G=1
    InColorAndOpacity_B=1
    InColorAndOpacity_A=1
   )
   > SetColorAndOpacity(
    self=Get AmmoBar() -> (AmmoBar:Image Object Reference)
    InColorAndOpacity_R=1
    InColorAndOpacity_G=1
    InColorAndOpacity_B=1
    InColorAndOpacity_A=1
   )
   > SetLockedOnAmmo!(
    execute
    self
    InCount:int=GetCurrentLockOnCount!(self:Auto Aim Weapon Pawn Component Object Reference=Get AutoAimComponent()) -> (ReturnValue:int=0)
   ) -> (then)
   SetColorAndOpacity(execute=@n5.else, self=Get ReticleVisual(), InColorAndOpacity_R=0, InColorAndOpacity_G=0, InColorAndOpacity_B=0, InColorAndOpacity_A=0)
   > SetColorAndOpacity(
    self=[Get LockedOnAmmo_1(), Get LockedOnAmmo(), Get LockedOnAmmo_2(), Get LockedOnAmmo_3()]
    InColorAndOpacity_R=0
    InColorAndOpacity_G=0
    InColorAndOpacity_B=0
    InColorAndOpacity_A=0
   )
   > SetLockedOnAmmo(InCount=0)
   > SetColorAndOpacity(self=Get AmmoBar(), InColorAndOpacity_R=0, InColorAndOpacity_G=0, InColorAndOpacity_B=0, InColorAndOpacity_A=0)
   Event Destruct()
   n6 = Get As B Pistol Smart Final()
   > Is Valid(InputObject=@n6)
   > Unbind all Events from On Local Ammo Changed!(execute:Exec, self:Fort Weapon Object Reference=@n6) -> (then:Exec) {is=title K2Node_ClearDelegate}
   Event OnAmmoClipAttributeChanged!() -> (OutputDelegate, then, AmmoClipValue:float) {is=title K2Node_Event}
   n7 = Get As B Pistol Smart Final()
   > Is Valid(InputObject=@n7)
   > ResetAmmo(LocalCount=GetMagazineAmmoCount(self=@n7) -> (ReturnValue=0), LocalRemaining=GetRemainingAmmo(self=@n7) -> (ReturnValue=0))
  ]}
 GainedLockTarget {nodes [
   GainedLockTarget() {is=title K2Node_FunctionEntry}
   > PlayAnimationForward(InAnimation=Get OnLockOn() -> (OnLockOn:Widget Animation Object Reference))
  ]}
 LostLockTarget {nodes [
   LostLockTarget() {is=title K2Node_FunctionEntry}
   > PlayAnimationReverse(InAnimation=Get OnLockOn())
  ]}
 OnFullyCharged {nodes [
   OnFullyCharged() {is=title K2Node_FunctionEntry}
   > PlayAnimation(InAnimation=Get FinalLock() -> (FinalLock:Widget Animation Object Reference))
  ]}
 OnLocalAmmoChanged {nodes [
   n1 = OnLocalAmmoChanged!() -> (then, LocalCount:int, LocalRemaining:int) {is=title K2Node_FunctionEntry}
   > ResetAmmo(LocalCount=@n1, LocalRemaining=@n1.LocalRemaining)
  ]}
 OnLockAllowed {nodes [OnLockAllowed() {is=title K2Node_FunctionEntry}]}
 OnPaint {nodes [
   n1 = Cast To FortPlayerPawn(Object=Get Instigator!(self:Actor Object Reference=GetEquippedWeapon()) -> (Instigator:Pawn Object Reference))
   > n2 = Cast To FortPlayerController(Object=GetController!(self:Pawn Object Reference=@n1) -> (ReturnValue:Controller Object Reference))
   > Paint Circle!(
    execute
    InContext:Paint Context Structure=OnPaint!() -> (then, Context:"Paint Context Structure (by ref)") {is=title K2Node_FunctionEntry}
    HasTarget:bool=false
    InRadius:double=20
    FOV:double=GetFOVAngle!(self:Player Camera Manager Object Reference=Get PlayerCameraManager!(self:Player Controller Object Reference=@n2) -> (PlayerCameraManager:Player Camera Manager Object Reference)) -> (ReturnValue:float)
    GeoHalfSize:double=0
   ) -> (then)
   > Return Node()
  ]}
 ResetAmmo {nodes [
   n1 = ResetAmmo!() -> (then, LocalCount:int, LocalRemaining:int) {is=title K2Node_FunctionEntry}
   > Set TotalAmmo(TotalAmmo:int=@n1.LocalRemaining) -> (Output_Get:int=0)
   > n2 = GetDynamicMaterial(self=Get AmmoBar())
   n3 = GetBulletsPerClip!(self:Fort Weapon Object Reference=Get As B Pistol Smart Final()) -> (ReturnValue:int=0)
   n4 = Get LocalCount() -> (LocalCount:int=0)
   > SetScalarParameterValue(self=@n2, ParameterName=Progress, Value=float / float(A="To Float (Integer)"(InInt=@n4), B="To Float (Integer)"(InInt=@n3)))
   > Set TotalAmmo(TotalAmmo=int - int(A=Get TotalAmmo() -> (TotalAmmo:int=0), B=Get LocalCount() -> (LocalCount=0))) -> (Output_Get=0)
  ]}
 SetLockedOnAmmo {nodes [
   SetLockedOnAmmo!() -> (then, InCount:int) {is=title K2Node_FunctionEntry}
   > n1 = For Loop(LastIndex=3)
   > Set CurrIdx(CurrIdx:int=@n1) -> (Output_Get:int=0)
   n2 = Get InCount() -> (InCount:int=0)
   n3 = Get CurrIdx() -> (CurrIdx:int=0)
   > SetBrushFromMaterial(
    self="Get (a copy)"(Array:Array of Image Object References=Get AmmoList(), Dimension 1=Get CurrIdx() -> (CurrIdx=0)) -> (Output:Image Object Reference)
    Material=SelectMaterialInstance!(
     A:Material Instance Object Reference=/Smartgun/Reticle/Materials/MI_UI_SmartPistolStack_Filled
     B:Material Instance Object Reference=/Smartgun/Reticle/Materials/MI_UI_SmartPistolStack_Empty
     bSelectA:bool="integer > integer"(A=@n2, B=@n3)
    ) -> (ReturnValue:Material Instance Object Reference)
   )
  ]}
 TickSpread {nodes [
   Set SpreadDeg(SpreadDeg:double=TickSpread!() -> (then, SpreadDegrees:double=~, InGeo:Geometry Structure) {is=title K2Node_FunctionEntry}) -> (Output_Get:double)
   n1 = Get SpreadDeg() -> (SpreadDeg:double)
   > Set ReticleTan(ReticleTan:double="Tan (Radians)"!(A:double=Degrees To Radians!(A:double=@n1) -> (ReturnValue:double)) -> (ReturnValue:double)) -> (Output_Get:double)
   n2 = GetEquippedWeapon()
   > Is Valid(InputObject=@n2)
   n3 = Get Instigator(self=@n2)
   > Is Valid(InputObject=@n3)
   n4 = GetFortPlayerControllerFromActor!(Actor:Actor Object Reference=@n3) -> (ReturnValue:Fort Player Controller Object Reference)
   > n5 = Set FOV(FOV:double=GetFOVAngle(self=Get PlayerCameraManager(self=@n4))) -> (Output_Get:double)
   n6 = float * float(A=@n5, B=0.5) {is=title K2Node_PromotableOperator}
   > Set FOV_Tan(FOV_Tan:double="Tan (Radians)"(A=Degrees To Radians(A=@n6))) -> (Output_Get:double)
   n7 = Get FOV_Tan() -> (FOV_Tan:double)
   n8 = Get FOV_Tan()
   n9 = Get ReticleTan() -> (ReticleTan:double)
   > Set ReticleToScreenTanRatio(ReticleToScreenTanRatio:double=SelectFloat(A=0, B=float / float(A=@n9, B=@n8), bPickA="Nearly Equal (Float)"(A=@n7, B=0, ErrorTolerance=0))) -> (Output_Get:double)
   n10 = GetViewportSize!() -> (ReturnValue:Vector 2D Structure)
   > Set ReticleRadius(ReticleRadius:double=float * float!(A=BreakVector2D(InVec=@n10), B=Get ReticleToScreenTanRatio() -> (ReticleToScreenTanRatio:double), C:double=0.5) -> (ReturnValue)) -> (Output_Get:double)
   n11 = float + float(A=Get ReticleRadius() -> (ReticleRadius:double), B=5)
   > SetRenderTranslation(self=Get Left() -> (Left:Image Object Reference), Translation=MakeVector2D(X=float * float!(A=@n11, B=-1) -> (ReturnValue), Y=0))
   n12 = float + float(A=Get ReticleRadius(), B=5)
   > SetRenderTranslation(self=Get Right() -> (Right:Image Object Reference), Translation=MakeVector2D(X=float * float(A=@n12, B=1), Y=0))
   n13 = Get ReticleRadius()
   > SetRenderTranslation(self=Get Bottom() -> (Bottom:Image Object Reference), Translation=MakeVector2D(X=0, Y=float + float(A=@n13, B=5)))
   n14 = float + float(A=Get ReticleRadius(), B=5)
   > SetRenderTranslation(self=Get Top() -> (Top:Image Object Reference), Translation=MakeVector2D(X=0, Y=float * float(A=@n14, B=-1)))
   GetLocalSize!(Geometry:"Geometry Structure (by ref)"=Get InGeo() -> (InGeo:Geometry Structure)) -> (ReturnValue:Vector 2D Structure)
   float * float(B=0.5)
  ]}
 UpdateBinding {nodes [
   UpdateBinding() {is=title K2Node_FunctionEntry}
   n1 = Get AutoAimComponent()
   > n2 = Branch(Condition=AND Boolean(A=Get DidBind() -> (DidBind:bool), B=IsValid(Object=@n1)))
   n3 = GetEquippedWeapon()
   Is Valid(exec=@n2.else, InputObject=@n3)
   n4 = Get Instigator(self=@n3)
   > Is Valid(InputObject=@n4)
   n5 = GetComponentByClass!(
    self:Actor Object Reference=@n4
    ComponentClass:Actor Component Class Reference=/Smartgun/Gameplay/StackingLockons/AutoAim_Pistol_Component.AutoAim_Pistol_Component_C
   ) -> (ReturnValue:Auto Aim Pistol Component Object Reference)
   > Is Valid(InputObject=@n5)
   > Set AutoAimComponent(AutoAimComponent:Auto Aim Pistol Component Object Reference=@n5) -> (Output_Get:Auto Aim Pistol Component Object Reference)
   > Bind Event to On Full Charge(self:Auto Aim Pistol Component Object Reference=Get AutoAimComponent(), Delegate=Create Event()) {is=title K2Node_AddDelegate}
   > Set DidBind(DidBind:bool=true) -> (Output_Get:bool)
   > Bind Event to On Gained Lock Target(self:Auto Aim Pistol Component Object Reference=Get AutoAimComponent(), Delegate=Create Event()) {is=title K2Node_AddDelegate}
   > Bind Event to On Lost Lock Target(self:Auto Aim Pistol Component Object Reference=Get AutoAimComponent(), Delegate=Create Event()) {is=title K2Node_AddDelegate}
   > Bind Event to On Lockon Allowed(self:Auto Aim Pistol Component Object Reference=Get AutoAimComponent(), Delegate=Create Event()) {is=title K2Node_AddDelegate}
  ]}
]
variables [
 AmmoList {type="array<Image>", default []}
 As B Pistol Smart Final {type=B_AutoAim_Pistol_C, default=null}
 AutoAimComponent {type=AutoAim_Pistol_Component_C, default=null}
 DidBind {type=boolean}
 FOV {type=number}
 FOV_Tan {type=number}
 HasTargetCantFire {type=LinearColor, default=<rgba 0.729 0.7552 0.062 1>}
 HasTargetColor {type=LinearColor, default=<rgba 1 0 0 1>}
 NoTargetColor {type=LinearColor, default=<rgba 1 1 1 1>}
 ReticleRadius {type=number}
 ReticleTan {type=number}
 ReticleToScreenTanRatio {type=number}
 SpreadDeg {type=number}
 TotalAmmo {type=integer}
]
widgets {
 info {
  inheritedWidgetCount=1
  parentClass=/Script/FortniteUI.FortWeaponReticleExtensionWidgetBase
  rootWidgetClass=&Overlay
  widgetCount=11
 }
 widgets [
  Overlay_18 {widgetClassPath=&Overlay, widget=Overlay_18}
  AmmoBar {
   widgetClassPath=&Image
   bIsVariable=true
   parent=Overlay_18
   widget=AmmoBar
   properties {
    brush {
     bIsDynamicallyLoaded=false
     drawAs=Image
     imageType=NoImage
     mirroring=NoMirror
     resourceName=None
     resourceObject=/Smartgun/Reticle/Materials/MI_UI_SmartPistolAmmoBar
     tiling=NoTile
     imageSize=<250 250>
     margin=<ltrb 0 0 0 0>
     outlineSettings {
      bUseBrushTransparency=false
      roundingType=HalfHeightRadius
      width=0
      color {colorUseRule=UseColor_Specified, specifiedColor=<rgba 0 0 0 0>}
      cornerRadii=<0 0 0 0>
     }
     tintColor {colorUseRule=UseColor_Specified, specifiedColor=<rgba 1 1 1 0.6>}
     uVRegion {bIsValid=false, max=<0 0>, min=<0 0>}
    }
    colorAndOpacity=<rgba 1 1 1 0.65>
    renderTransform {angle=65, scale=<1 1>, shear=<0 0>, translation=<198 -40>}
    renderTransformPivot=<0 0>
   }
   slot {class=OverlaySlot, properties {horizontalAlignment=HAlign_Center, verticalAlignment=VAlign_Center}}
  }
  Overlay_0 {
   widgetClassPath=&Overlay
   parent=Overlay_18
   widget=Overlay_0
   slot {class=OverlaySlot, properties {horizontalAlignment=HAlign_Fill, verticalAlignment=VAlign_Fill}}
  }
  LockedOnAmmo {
   widgetClassPath=&Image
   bIsVariable=true
   parent=Overlay_0
   widget=LockedOnAmmo
   properties {
    brush {
     bIsDynamicallyLoaded=false
     drawAs=Image
     imageType=NoImage
     mirroring=NoMirror
     resourceName=None
     resourceObject=null
     tiling=NoTile
     imageSize=<28 22>
     margin=<ltrb 0 0 0 0>
     outlineSettings {
      bUseBrushTransparency=false
      roundingType=HalfHeightRadius
      width=0
      color {colorUseRule=UseColor_Specified, specifiedColor=<rgba 0 0 0 0>}
      cornerRadii=<0 0 0 0>
     }
     tintColor {colorUseRule=UseColor_Specified, specifiedColor=<rgba 1 1 1 1>}
     uVRegion {bIsValid=false, max=<0 0>, min=<0 0>}
    }
    renderTransform {angle=22, scale=<0.8 0.8>, shear=<0 0>, translation=<-5.5 38>}
   }
   slot {class=OverlaySlot, properties {horizontalAlignment=HAlign_Right, verticalAlignment=VAlign_Center, padding=<ltrb 0 0 69 0>}}
  }
  LockedOnAmmo_1 {
   widgetClassPath=&Image
   bIsVariable=true
   parent=Overlay_0
   widget=LockedOnAmmo_1
   properties {
    brush {
     bIsDynamicallyLoaded=false
     drawAs=Image
     imageType=NoImage
     mirroring=NoMirror
     resourceName=None
     resourceObject=null
     tiling=NoTile
     imageSize=<28 22>
     margin=<ltrb 0 0 0 0>
     outlineSettings {
      bUseBrushTransparency=false
      roundingType=HalfHeightRadius
      width=0
      color {colorUseRule=UseColor_Specified, specifiedColor=<rgba 0 0 0 0>}
      cornerRadii=<0 0 0 0>
     }
     tintColor {colorUseRule=UseColor_Specified, specifiedColor=<rgba 1 1 1 1>}
     uVRegion {bIsValid=false, max=<0 0>, min=<0 0>}
    }
    renderTransform {angle=6, scale=<0.8 0.8>, shear=<0 0>, translation=<10 9>}
   }
   slot {class=OverlaySlot, properties {horizontalAlignment=HAlign_Right, verticalAlignment=VAlign_Center, padding=<ltrb 0 0 78 0>}}
  }
  LockedOnAmmo_2 {
   widgetClassPath=&Image
   bIsVariable=true
   parent=Overlay_0
   widget=LockedOnAmmo_2
   properties {
    brush {
     bIsDynamicallyLoaded=false
     drawAs=Image
     imageType=NoImage
     mirroring=NoMirror
     resourceName=None
     resourceObject=null
     tiling=NoTile
     imageSize=<28 22>
     margin=<ltrb 0 0 0 0>
     outlineSettings {
      bUseBrushTransparency=false
      roundingType=HalfHeightRadius
      width=0
      color {colorUseRule=UseColor_Specified, specifiedColor=<rgba 0 0 0 0>}
      cornerRadii=<0 0 0 0>
     }
     tintColor {colorUseRule=UseColor_Specified, specifiedColor=<rgba 1 1 1 1>}
     uVRegion {bIsValid=false, max=<0 0>, min=<0 0>}
    }
    renderTransform {angle=10, scale=<0.8 0.8>, shear=<0 0>, translation=<5 19>}
   }
   slot {class=OverlaySlot, properties {horizontalAlignment=HAlign_Right, verticalAlignment=VAlign_Center, padding=<ltrb 0 0 74 0>}}
  }
  LockedOnAmmo_3 {
   widgetClassPath=&Image
   bIsVariable=true
   parent=Overlay_0
   widget=LockedOnAmmo_3
   properties {
    brush {
     bIsDynamicallyLoaded=false
     drawAs=Image
     imageType=NoImage
     mirroring=NoMirror
     resourceName=None
     resourceObject=null
     tiling=NoTile
     imageSize=<28 22>
     margin=<ltrb 0 0 0 0>
     outlineSettings {
      bUseBrushTransparency=false
      roundingType=HalfHeightRadius
      width=0
      color {colorUseRule=UseColor_Specified, specifiedColor=<rgba 0 0 0 0>}
      cornerRadii=<0 0 0 0>
     }
     tintColor {colorUseRule=UseColor_Specified, specifiedColor=<rgba 1 1 1 1>}
     uVRegion {bIsValid=false, max=<0 0>, min=<0 0>}
    }
    renderTransform {angle=15, scale=<0.8 0.8>, shear=<0 0>, translation=<0 29>}
   }
   slot {class=OverlaySlot, properties {horizontalAlignment=HAlign_Right, verticalAlignment=VAlign_Center, padding=<ltrb 0 0 71 0>}}
  }
  Overlay {widgetClassPath=&Overlay, parent=Overlay_18, widget=Overlay, slot {class=OverlaySlot}}
  ReticleVisual {
   widgetClassPath=&Image
   bIsVariable=true
   parent=Overlay
   widget=ReticleVisual
   properties {
    brush {
     bIsDynamicallyLoaded=false
     drawAs=Image
     imageType=NoImage
     mirroring=NoMirror
     resourceName=None
     resourceObject=/Smartgun/Reticle/Materials/MI_UI_SmartPistolCircle
     tiling=NoTile
     imageSize=<380 380>
     margin=<ltrb 0 0 0 0>
     outlineSettings {
      bUseBrushTransparency=false
      roundingType=HalfHeightRadius
      width=0
      color {colorUseRule=UseColor_Specified, specifiedColor=<rgba 0 0 0 0>}
      cornerRadii=<0 0 0 0>
     }
     tintColor {colorUseRule=UseColor_Specified, specifiedColor=<rgba 1 1 1 1>}
     uVRegion {bIsValid=false, max=<0 0>, min=<0 0>}
    }
    colorAndOpacity=<rgba 1 1 1 0.3>
   }
   slot {class=OverlaySlot, properties {horizontalAlignment=HAlign_Fill, verticalAlignment=VAlign_Fill}}
  }
  Center {
   widgetClassPath=&Image
   bIsVariable=true
   parent=Overlay_18
   widget=Center
   properties {
    brush {
     bIsDynamicallyLoaded=false
     drawAs=Image
     imageType=NoImage
     mirroring=NoMirror
     resourceName=None
     resourceObject=/Game/UI/HUD/Reticles/HUD_Xhair_Shotgun_CntDot
     tiling=NoTile
     imageSize=<4 4>
     margin=<ltrb 0 0 0 0>
     outlineSettings {
      bUseBrushTransparency=false
      roundingType=HalfHeightRadius
      width=0
      color {colorUseRule=UseColor_Specified, specifiedColor=<rgba 0 0 0 0>}
      cornerRadii=<0 0 0 0>
     }
     tintColor {colorUseRule=UseColor_Specified, specifiedColor=<rgba 1 1 1 1>}
     uVRegion {bIsValid=false, max=<0 0>, min=<0 0>}
    }
    colorAndOpacity=<rgba 0 0.6378 1 0.7>
   }
   slot {class=OverlaySlot, properties {horizontalAlignment=HAlign_Center, verticalAlignment=VAlign_Center}}
  }
  TutorialHighlight {widgetClassPath=&TutorialHighlight}
 ]
}
